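# Single Layer Water material in the MaterialGym, filed under a translucency sorting test for that shading model.
# Color and strength parameters multiply into scattering and absorption (murkiness) coefficients, each divided by 100 before feeding the Single Layer Water output.
# Phase G and color scale behind water have nothing wired in.
# Refraction mode is pixel normal offset, and a Fresnel-driven lerp from 1 to an IOR parameter is present, with no connection shown.
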Material /Game/Global/Materials/MaterialGym/Materials/SortTranslucentSingleLayerWater/Materials/M_SingleLayerWater
graph {
 properties {
  bAutomaticallySetUsageInEditor=false
  refractionMethod=RM_PixelNormalOffset
  shadingModel=MSM_SingleLayerWater
  usedShadingModels=SingleLayerWater
 }
 nodes [
  LinearInterpolate(B=ScalarParameter() {defaultValue=1.333, parameterName=IOR}, Alpha=Fresnel()) {constA=1}
  n1 = VectorParameter() {parameterName=WaterScatteringColor, defaultValue=<rgba 0.1135 0.4976 1 1>}
  n2 = ScalarParameter() {defaultValue=0.01, parameterName=WaterScattering}
  n3 = VectorParameter() {parameterName=WaterAbsorption, defaultValue=<rgba 1 0.5685 0.0978 1>}
  n4 = ScalarParameter() {defaultValue=1.35, parameterName=WaterMurkiness}
  SingleLayerWaterMaterialOutput!(
   ScatteringCoefficients=Divide(A=Multiply(A=@n1.RGB, B=@n2)) {constB=100}
   AbsorptionCoefficients=Divide(A=Multiply(A=@n3.RGB, B=@n4)) {constB=100}
   PhaseG
   ColorScaleBehindWater
  )
 ]
}
