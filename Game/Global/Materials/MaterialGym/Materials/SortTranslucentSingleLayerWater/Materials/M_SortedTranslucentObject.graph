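# Material gym translucent lit material for an object that sorts correctly against Single Layer Water, staying visible below its surface.
# With depth testing disabled, a Custom Depth versus pixel depth comparison masks out the object's own hidden surfaces.
# A Scene Depth comparison and ShadingModelID check fade opacity with depth behind water, but hide pixels behind other opaque surfaces.
# A local-Z gradient blends bottom and top colors.
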Material /Game/Global/Materials/MaterialGym/Materials/SortTranslucentSingleLayerWater/Materials/M_SortedTranslucentObject
graph {
 properties {
  allowTranslucentCustomDepthWrites=true
  bAutomaticallySetUsageInEditor=false
  bDisableDepthTest=true
  blendMode=BLEND_Translucent
  usedShadingModels=Default Lit
 }
 nodes [
  n1 = Add(A=ComponentMask(None=SceneTexture() {sceneTextureId=PPI_CustomDepth}.Color) {r=true})
  n2 = PixelDepth()
  n3 = Multiply(
   A=ScalarParameter() {defaultValue=0.7, group=04_OpacityControl, parameterName=Opactiy_Value, sliderMax=1}
   B=Saturate(None=Subtract(A=@n1, B=@n2.R))
  )
  n4 = Subtract(A=PixelDepth().R, B=ComponentMask(None=SceneTexture() {sceneTextureId=PPI_SceneDepth}.Color) {r=true})
  n5 = Divide(A=@n4, B=ScalarParameter() {defaultValue=100, parameterName=UnderWater_DistanceBlend, sliderMax=1024})
  n6 = LinearInterpolate(
   A=ScalarParameter() {parameterName=UnderWater_MinOpacity, sliderMax=1}
   B=ScalarParameter() {defaultValue=0.8, parameterName=UnderWater_MaxOpacity, sliderMax=1}
   Alpha=OneMinus(None=Saturate(None=@n5))
  ) {constA=0.2, constB=0.2}
  n7 = Ceil(None=Saturate(None=@n4))
  n8 = ScreenPosition()
  n9 = Subtract(
   A=ComponentMask(None=SceneTexture(UVs=@n8.ViewportUV) {sceneTextureId=PPI_ShadingModelID}.Color) {r=true}
   B=Constant() {desc=ID of SingleLayerWater, r=10}
  )
  LinearInterpolate(
   A=LinearInterpolate(A=@n3, B=Multiply(A=@n3, B=@n6), Alpha=@n7) {desc=above and under water level}
   B=Multiply(A=@n3, B=OneMinus(None=@n7))
   Alpha=Saturate(None=Abs(None=@n9))
  ) {desc=infront or under water}
  n10 = Add(A=Divide(A=LocalPosition().Z) {constB=100}) {constB=0.5}
  LinearInterpolate(
   A=VectorParameter() {parameterName=Color_Bottom, defaultValue=<rgba 0.066 0.2948 0.0852 1>}.RGB
   B=VectorParameter() {parameterName=Color_Top, defaultValue=<rgba 0.8134 0.824 0.3446 1>}.RGB
   Alpha=Saturate(None=Power(Base=@n10) {constExponent=2.2})
  )
  n11 = Subtract(A=WorldPosition().XYZ, B=CameraPositionWS())
  Multiply(A=ComponentMask(None=Transform(None=@n11) {transformSourceType=TRANSFORMSOURCE_World, transformType=TRANSFORM_View}) {b=true}) {constB=0.9999}
 ]
}
